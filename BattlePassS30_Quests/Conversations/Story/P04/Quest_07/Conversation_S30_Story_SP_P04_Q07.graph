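# Season 30 story quest conversation (Part 4, Quest 7) in which the player gives Hope the Pandora's Gem.
# One entry point feeds three linear Speech nodes: greeting, gem handover and sign-off.
# The auto-selected main-questline greeting requires the NPC's VoidRedemption entry tag, the active quest and an incomplete obj0 via an inverted check.
# The "Give" choice carries the VoidRedemption quest gameplay tag that marks the handover.
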ConversationDatabase /BattlePassS30_Quests/Conversations/Story/P04/Quest_07/Conversation_S30_Story_SP_P04_Q07
entry_points [{tag=Conversation.Entry.Season30.Story.P04.Q07.VoidRedemption, destinations [#1]}]
nodes [
 ConversationEntryPointNode_0 {
  id=#1
  class=ConversationEntryPointNode
  outputs [#2]
  properties {entryTag {tagName=Conversation.Entry.Season30.Story.P04.Q07.VoidRedemption}}
 }
 Speech_C_1 {
  id=#2
  class=Speech_C
  outputs [#3]
  properties {softTaskIcon=&T_UI_NPC_MainQuest, generalConfig {defaultMessage=Unngh... It's happening again...}}
  sub_nodes [
   ConversationChoiceNode_2 {
    class=ConversationChoiceNode
    properties {
     defaultChoiceDisplayText=Talk to Hope
     choiceTags {gameplayTags [{tagName=Conversation.Choice.AutoSelectIfValid}, {tagName=Conversation.Choice.Category.MainQuestline}]}
    }
   }
   FortConversationRequirement_ControllerRequirement_1 {
    class=FortConversationRequirement_ControllerRequirement
    properties {
     nodeName=Require: NPC Has Conversation Tag
     participantId {tagName=Conversation.Participant.NPC}
     requirement {class=FortControllerRequirement_NPCHasConversationEntryTag, properties {queryTag {tagName=Conversation.Entry.VoidRedemption}}}
    }
   }
   FortConversationRequirement_ControllerRequirement_3 {
    class=FortConversationRequirement_ControllerRequirement
    properties {
     nodeName=Require: Is Quest Active
     requirement {
      class=FortControllerRequirement_IsQuestActive
      properties {questDefinition=/BattlePassS30_Quests/Items/QuestItems/Story/P04/Quest_S30_Story_P04_Q07}
     }
    }
   }
   FortConversationRequirement_ControllerRequirement_5 {
    class=FortConversationRequirement_ControllerRequirement
    properties {
     nodeName=Require: Has Completed Quest Objective
     requirement {
      class=FortControllerRequirement_HasCompletedQuestObjective
      properties {
       bInverseCondition=true
       objectiveBackendName=Quest_S30_Story_P04_Q07_obj0
       questDefinition=&Quest_S30_Story_P04_Q07
      }
     }
    }
   }
  ]
 }
 Speech_C_28 {
  id=#3
  class=Speech_C
  outputs [#4]
  properties {softTaskIcon=&T_UI_NPC_Conversation_Give, generalConfig {defaultMessage=The headache is stronger this time...}}
  sub_nodes [ConversationChoiceNode_1 {
    class=ConversationChoiceNode
    properties {
     defaultChoiceDisplayText=Give Hope the Pandora's Gem
     choiceTags {gameplayTags [{tagName=Athena.Quests.S30.Conversation.Story.P04.07.VoidRedemption}]}
    }
   }]
 }
 Speech_C_2 {
  id=#4
  class=Speech_C
  outputs []
  properties {generalConfig {defaultMessage=I'll... I'll be in touch. I need to relax. Maybe paint something...}}
  sub_nodes [ConversationChoiceNode_3 {class=ConversationChoiceNode, properties {defaultChoiceDisplayText=Continue}}]
 }
]
